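# Material function that builds the metal look for the Dark Storm Armored Super Levels (SL21) skin, overriding incoming material attributes.
# An FX mask texture blends base, red and blue tint colors shaded by the Gmap input, and also drives masks such as emissive.
# Thin-film color is blended by view reflection, and directional noise adds roughness scratches.
# Emissive is a blue-tinted Fresnel, pulsed and broken up by time-scrolled noise projected from pre-skinned position.
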MaterialFunction /Game/Characters/Player/Male/Medium/Bodies/M_MED_Dark_Storm_Armored/Skins/Super_Levels/Materials/MF_SK_S21_Metal
graph {nodes [
  Constant() {r=80}
  n1 = GetMaterialAttributes(None=FunctionInput() {inputName=Material, inputType=FunctionInput_MaterialAttributes}) {attributeGetTypes [#1, #2]}
  n2 = SetMaterialAttributes!(
   MaterialAttributes=@n1.MaterialAttributes
   BaseColor=NamedRerouteUsage()
   Metallic=NamedRerouteUsage()
   Roughness=NamedRerouteUsage()
   Specular=@n1.MaterialAttributes
   EmissiveColor=NamedRerouteUsage()
   Normal
  ) {attributeSetTypes [
    #3
    #4
    #5
    #2
    #6
    #1
   ]}
  FunctionOutput(None=@n2)
  n3 = TextureObject() {
   samplerType=SAMPLERTYPE_Masks
   texture=/Game/Characters/Player/Male/Medium/Bodies/M_MED_Dark_Storm_Armored/Skins/Super_Levels/Textures/T_M_MED_Dark_Storm_SuperLevels_Armored_Body_SL_FX
  }
  n4 = TextureSample(Tex=FunctionInput(Preview=@n3) {inputType=FunctionInput_Texture2D}) {texture=/Game/Characters/Player/Male/Medium/Bodies/M_MED_Dark_Storm/Texture/T_M_MED_Dark_Storm_Unarmored_Body_E}
  NamedRerouteDeclaration(None=@n4.B) {name=Blue, nodeColor=<rgba 1 0.0469 0 1>}
  NamedRerouteDeclaration(None=@n4.R) {name=Red, nodeColor=<rgba 1 0 0.2812 1>}
  NamedRerouteDeclaration(None=@n4.G) {name=green, nodeColor=<rgba 1 0 0.4219 1>}
  NamedRerouteDeclaration(None=LinearInterpolate(Alpha=NamedRerouteUsage()) {constA=1, constB=0}) {name=Metal, nodeColor=<rgba 1 0.0469 0 1>}
  n5 = Constant() {r=0.2}
  n6 = TextureSample() {
   samplerType=SAMPLERTYPE_LinearColor
   texture=/Game/Characters/Player/Male/Medium/Bodies/M_MED_Mitosis_Shadow/Textures/DirectionalNoise__Noise
  }
  n7 = LinearInterpolate(A=@n6.R, B=@n6.R, Alpha=NamedRerouteUsage())
  n8 = ScalarParameter() {defaultValue=0.4, parameterName=MetalSratchAmount}
  NamedRerouteDeclaration(None=LinearInterpolate(A=LinearInterpolate(A=@n5, B=@n7, Alpha=@n8), Alpha=NamedRerouteUsage()) {constB=0.3}) {name=Rough, nodeColor=<rgba 0 1 0.0156 1>}
  n9 = VectorParameter() {group=SL21, parameterName=MetalColor_Blue, defaultValue=<rgba 0.3772 0.6505 0.7031 1>}
  n10 = VertexInterpolator(VS=PreSkinnedPosition())
  n11 = MaterialFunctionCall(Float3=@n10.PS) {materialFunction=&BreakOutFloat3Components}
  n12 = SphereMask(A=@n11.B, B=Constant() {r=125}, Radius=Constant() {r=10}, Hardness=Constant())
  n13 = Subtract(A=@n11.B, B=Multiply(A=Time()) {constB=5})
  n14 = Abs(None=Constant() {r=64})
  n15 = ComponentMask(None=Divide(
    A=MaterialFunctionCall(X=@n11.R, Y=@n11.R, Z=@n13) {materialFunction=&MakeFloat3}.Result
    B=Multiply(A=@n14) {constB=-1}
   )) {b=true, r=true}
  n16 = Multiply(A=TextureCoordinate(), B=ScalarParameter() {defaultValue=0.25, parameterName=MipDerivativeScale})
  n17 = Add(
   A=Multiply(A=Saturate(None=@n12))
   B=TextureSample!(
    UVs=@n15
    Tex=TextureObject() {samplerType=SAMPLERTYPE_Masks, texture=/Game/Weapons/WeaponSkins/Wraps/Textures/Mash/Mash_Noise}
    "DDX(UVs)"=DDX(Value=@n16)
    "DDY(UVs)"=DDY(Value=@n16)
    Apply View MipBias
   ) {desc=Y projection, mipValueMode=TMVM_Derivative, samplerSource=SSM_Wrap_WorldGroupSettings}.R
  )
  n18 = Fresnel(ExponentIn=LinearInterpolate(A=Constant() {r=0.75}, Alpha=@n17) {constA=1, constB=1.4})
  n19 = Multiply(
   A=ComponentMask(None=@n17) {r=true}
   B=MaterialFunctionCall() {materialFunction=&DebugTimeSine}.Result
  )
  n20 = Clamp(None=Multiply(A=Power(Base=@n18), B=LinearInterpolate(Alpha=@n19) {constA=1, constB=3})) {minDefault=0.02}
  NamedRerouteDeclaration(None=LinearInterpolate(B=Multiply(A=@n9.RGB, B=@n20), Alpha=NamedRerouteUsage())) {name=Emissive, nodeColor=<rgba 0.5703 1 0 1>}
  n21 = Transform(None=@n1.Normal)
  NamedRerouteDeclaration(None=@n21) {name=Normal, nodeColor=<rgba 0.0078 1 0 1>}
  n22 = FunctionInput(Preview=Constant()) {inputName=Gmap, inputType=FunctionInput_Scalar}
  n23 = Multiply(A=Power(Base=@n22) {constExponent=3}) {constB=1.5}
  n24 = Multiply(
   A=LinearInterpolate(
    A=VectorParameter() {group=SL21, parameterName=MetalColor_Base_Occ, defaultValue=<rgba 0.7586 0.7429 0.9323 1>}.RGB
    B=VectorParameter() {group=SL21, parameterName=Metal_Base, defaultValue=<rgba 0.7 0.7 0.7 1>}.RGB
    Alpha=Saturate(None=@n23)
   )
   B=ScalarParameter() {defaultValue=1, group=SL21, parameterName=SL_BaseMult}
  )
  n25 = Multiply(
   A=Multiply(A=VectorParameter() {group=SL21, parameterName=MetalColor_Red, defaultValue=<rgba 0.7586 0.7429 0.9323 1>}.RGB, B=@n22)
   B=ScalarParameter() {defaultValue=1, group=SL21, parameterName=SL_RedMaskMult}
  )
  n26 = NamedRerouteUsage()
  n27 = Multiply(A=@n9.RGB, B=@n22)
  n28 = ScalarParameter() {defaultValue=1, group=SL21, parameterName=SL_BlueMaskMult, sortPriority=35}
  n29 = Multiply(
   A=Constant() {r=1}
   B=LinearInterpolate(A=LinearInterpolate(A=@n24, B=@n25, Alpha=@n26), B=Multiply(A=@n27, B=@n28), Alpha=NamedRerouteUsage())
  )
  n30 = Multiply(A=@n29) {constB=2}
  n31 = TextureSampleParameter2D(UVs=@n15) {group=SL21, parameterName=MetalThinFilmTexture, texture=/Game/Global/Textures/Default/Thinfilm/T_ThinFilm_Spectrum_DESATURATED-50}
  n32 = ScalarParameter() {defaultValue=0.4, group=SL21, parameterName=MetalThinFilmAmount}
  n33 = DotProduct(A=ReflectionVectorWS(CustomWorldNormal=@n21), B=CameraVectorWS())
  NamedRerouteDeclaration(None=LinearInterpolate(A=@n29, B=LinearInterpolate(A=@n30, B=@n31.RGB, Alpha=@n32) {constA=1}, Alpha=Saturate(None=@n33))) {name=Base Color, nodeColor=<rgba 1 0.8672 0 1>}
  NamedRerouteDeclaration(None=ChannelMaskParameter(None=@n4.RGB) {group=SL21, maskChannel=Blue, parameterName=MetalEmissiveMask, defaultValue=<rgba 0 0 1 0>}.R) {name=EmissiveMask, nodeColor=<rgba 0.3125 1 0 1>}
  NamedRerouteDeclaration(None=@n20) {name=EdgeFresnel, nodeColor=<rgba 1 0.7266 0 1>}
  NamedRerouteUsage()
  NamedRerouteUsage()
  Transform(None=@n10.PS) {transformSourceType=TRANSFORMSOURCE_Local}
 ]}
