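# MF_SK_S21_Reality is a material function that adds an animated Reality glow to Season 21 Super Level skins by blending new emissive into incoming material attributes.
# An input texture's RGB channels provide masks, while vector noise on pre-skinned position supplies distortion.
# Panning axis-projected textures blended by normal combine with parallax tendrils, tinted by scrolling curve-atlas gradients selected by SL_Tier.
# Static switches adapt it for heads and Blue Jay bodies.
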MaterialFunction /Game/Characters/Player/Male/Medium/Bodies/M_MED_Dark_Storm_Armored/Skins/Super_Levels/Materials/MF_SK_S21_Reality
graph {nodes [
  n1 = TextureCoordinate()
  Add(A=@n1, B=Multiply(A=NamedRerouteUsage()) {constB=0})
  n2 = GetMaterialAttributes(None=FunctionInput() {inputName=Material, inputType=FunctionInput_MaterialAttributes}) {attributeGetTypes [#1]}
  n3 = NamedRerouteUsage()
  n4 = NamedRerouteUsage()
  FunctionOutput(None=SetMaterialAttributes!(MaterialAttributes=@n2.MaterialAttributes, EmissiveColor=LinearInterpolate(A=@n2.EmissiveColor, B=@n3, Alpha=@n4)) {attributeSetTypes [#1]})
  n5 = Divide(A=NamedRerouteUsage(), B=Constant() {r=50})
  n6 = Add(
   A=Multiply(A=Time(), B=VectorParameter() {group=PanningEmissive, parameterName=TriPlanner Speed, defaultValue=<rgba 0.01 0.02 -0.05 1>}.RGB)
   B=Multiply(A=NamedRerouteUsage()) {constB=0.01}
  ) {constB=0}
  n7 = Add(A=@n5, B=@n6)
  n8 = TextureObjectParameter() {
   group=PanningEmissive
   parameterName=EmissiveGradient
   samplerType=SAMPLERTYPE_LinearColor
   texture=/Game/Environments/Artemis/Props/Daisy/Daisy_RealityGlitch/Textures/T_Tendrils
  }
  n9 = Add(A=@n5, B=@n6)
  n10 = VertexInterpolator(VS=PreSkinnedNormal())
  n11 = MaterialFunctionCall(
   In=Abs(None=ComponentMask(None=@n10.PS) {g=true})
   Contrast=ScalarParameter() {defaultValue=1, group=PanningEmissive, parameterName=TriPlanner_BlendStrength}
  ) {materialFunction=&CheapContrast}
  n12 = LinearInterpolate(
   A=TextureSample(UVs=ComponentMask(None=@n7) {b=true, g=true}, Tex=@n8) {samplerType=SAMPLERTYPE_Masks, texture=/Game/Weapons/WeaponSkins/Textures/T_Swirly_Noise_highContrast}.RGB
   B=TextureSample(UVs=ComponentMask(None=@n9) {b=true, r=true}, Tex=@n8) {samplerType=SAMPLERTYPE_Masks, texture=&T_Swirly_Noise_highContrast}.RGB
   Alpha=Saturate(None=ComponentMask(None=@n11.Result) {g=true})
  )
  n13 = Add(A=@n5, B=@n6)
  n14 = MaterialFunctionCall(In=Abs(None=ComponentMask(None=@n10.PS) {b=true}), Contrast=Constant() {r=1}) {materialFunction=&CheapContrast}
  LinearInterpolate(
   A=@n12
   B=TextureSample(UVs=ComponentMask(None=@n13) {g=true, r=true}, Tex=@n8) {samplerType=SAMPLERTYPE_Masks, texture=&T_Swirly_Noise_highContrast}.RGB
   Alpha=Saturate(None=ComponentMask(None=@n14.Result) {b=true})
  )
  n15 = Multiply(A=Time()) {constB=0.1}
  n16 = PreSkinnedLocalBounds()
  n17 = Divide(A=NamedRerouteUsage(), B=Subtract(A=@n16.Max, B=@n16.Max))
  n18 = ConstantBiasScale(None=Multiply(A=MaterialFunctionCall(Float3=@n17) {materialFunction=&BreakOutFloat3Components}.R))
  n19 = Frac(None=StaticSwitchParameter(True=@n15, False=Add(A=@n18, B=@n15)) {parameterName=IsHead})
  n20 = CurveAtlasRowParameter(CurveTime=@n19) {
   atlas=/Game/Athena/Apollo/Environments/Landscape/Materials/Frontier/FrontierColorAtlas
   curve=&/Curve_RealityColorBlend_T1
   defaultValue=6
   parameterName=RealityColor_T1
   sortPriority=0
  }
  MaterialFunctionCall(Hue Shift Percentage=ScalarParameter() {group=SL21, parameterName=RealityHueShift}, Texture=@n20) {materialFunction=&HueShift}
  MaterialFunctionCall(Float2=TextureCoordinate()) {materialFunction=&BreakOutFloat2Components}
  n21 = TextureObject() {
   samplerType=SAMPLERTYPE_Masks
   texture=/Game/Characters/Player/Female/Medium/Bodies/F_MED_Blue_Jay/Skins/Super_Levels/T_F_MED_Blue_Jay_Super_Levels_Body_Cat_SL_FX
  }
  n22 = TextureSample(Tex=FunctionInput(Preview=@n21) {inputType=FunctionInput_Texture2D}) {texture=/Game/Characters/Player/Male/Medium/Bodies/M_MED_Dark_Storm/Texture/T_M_MED_Dark_Storm_Unarmored_Body_E}
  NamedRerouteDeclaration(None=@n22.B) {name=Blue, nodeColor=<rgba 1 0.0469 0 1>}
  NamedRerouteDeclaration(None=@n22.R) {name=Red, nodeColor=<rgba 0.125 0 1 1>}
  n23 = MaterialFunctionCall(Float3=Divide(A=NamedRerouteUsage(), B=Constant() {r=20})) {materialFunction=&BreakOutFloat3Components}
  n24 = MaterialFunctionCall(X=@n23.R, Y=@n23.R, Z=Subtract(A=@n23.B, B=Time())) {materialFunction=&MakeFloat3}
  n25 = InverseLinearInterpolate(Value=ComponentMask(None=VectorNoise(World Position=@n24.Result) {noiseFunction=VNF_VectorALU}) {r=true}) {constB=10}
  n26 = ScalarParameter() {defaultValue=2, group=SL21, parameterName=DistortionMult}
  NamedRerouteDeclaration(None=Subtract(A=Multiply(A=@n25, B=@n26))) {name=Distortion, nodeColor=<rgba 1 0.7031 0 1>}
  n27 = CurveAtlasRowParameter(CurveTime=@n19) {
   atlas=&FrontierColorAtlas
   curve=&/Curve_RealityColorBlend_T2
   defaultValue=7
   parameterName=RealityColor_T2
   sortPriority=0
  }
  n28 = ScalarParameter() {defaultValue=1, group=SL21, parameterName=SL_Tier}
  n29 = Subtract(A=@n28)
  n30 = LinearInterpolate(A=@n20, B=@n27, Alpha=Saturate(None=Sign(None=@n29)))
  n31 = CurveAtlasRowParameter(CurveTime=@n19) {
   atlas=&FrontierColorAtlas
   curve=&/Curve_RealityColorBlend_T3
   defaultValue=8
   parameterName=RealityColor_T3
   sortPriority=0
  }
  n32 = Saturate(None=Sign(None=Subtract(A=@n28) {constB=2}))
  NamedRerouteDeclaration(None=Saturate(None=LinearInterpolate(A=@n30, B=@n31, Alpha=@n32))) {name=Color Tint, nodeColor=<rgba 1 0.8906 0 1>}
  NamedRerouteDeclaration(None=@n22.G) {name=Green, nodeColor=<rgba 1 0.2812 0 1>}
  NamedRerouteDeclaration(None=@n25) {name=Distortion2, nodeColor=<rgba 0.8281 0 1 1>}
  n33 = NamedRerouteUsage()
  n34 = Multiply(A=@n33) {constB=6}
  n35 = NamedRerouteUsage()
  n36 = Panner(Coordinate=@n1) {speedY=0.02}
  n37 = Constant() {r=0.1}
  n38 = Constant() {r=0.5}
  n39 = Power(Base=TextureSample(UVs=BumpOffset(Coordinate=@n36, Height=@n37, HeightRatioInput=@n38)) {samplerType=SAMPLERTYPE_LinearColor, texture=/Game/Environments/Artemis/Props/Daisy/Daisy_RealityGlitch/Textures/T_Tendrils_b}.R) {constExponent=1}
  n40 = NamedRerouteUsage()
  n41 = Multiply(A=Add(A=Multiply(A=@n39), B=Power(Base=@n40) {constExponent=3}), B=@n33)
  n42 = SphereMask(
   A=TextureCoordinate()
   B=StaticSwitchParameter(True=Constant2Vector() {g=0.25, r=0.59}, False=Constant2Vector() {g=0.36, r=0.53}) {parameterName=BlueJaySummerBody}
   Radius=Constant() {r=0.075}
  )
  n43 = ScalarParameter() {defaultValue=3, parameterName=FinalEmissiveMult}
  n44 = Multiply(
   A=StaticSwitchParameter(True=LinearInterpolate(A=@n12, Alpha=@n42) {constB=0}, False=@n12) {parameterName=BlueJayBody}
   B=LinearInterpolate(A=@n43, B=Multiply(A=@n43) {constB=3}, Alpha=NamedRerouteUsage())
  ) {constB=5}
  NamedRerouteDeclaration(None=StaticSwitchParameter(True=LinearInterpolate(B=@n34, Alpha=@n35), False=Add(A=@n41, B=@n44)) {parameterName=IsHead}) {name=Emissive, nodeColor=<rgba 0.6172 0 1 1>}
  n45 = FunctionInput(Preview=PreSkinnedPosition()) {inputName=PreskinnedPosition}
  NamedRerouteDeclaration(None=ShaderStageSwitch(PixelShader=VertexInterpolator(VS=@n45).PS, VertexShader=@n45)) {name=PreskinnedPosition, nodeColor=<rgba 1 0.0469 0 1>}
  Saturate(None=Subtract(A=@n4))
  ScalarParameter() {defaultValue=3, parameterName=FinalEmissiveMult}
  TextureSample() {
   samplerType=SAMPLERTYPE_Masks
   texture=&T_F_MED_Blue_Jay_Super_Levels_Body_Cat_SL_FX
  }
 ]}
