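# Master character material for Season 21 Super Level skins, layering metal and reality effects over the base character shading.
# The PreFX result feeds S21 Metal then Reality functions, both driven by the SLMask texture.
# A "Lancelot Cape" static switch flattens normals by the mask's blue channel.
# The M mask's blue channel, scaled by killskin, blends back to the PreFX look before the shared PostFX function.
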Material /Game/Characters/Player/Male/Medium/Bodies/M_MED_Dark_Storm_Armored/Skins/Super_Levels/Materials/M_SL_S21_Master
graph {
 properties {
  bUseMaterialAttributes=true
  bUsedWithClothing=true
  bUsedWithInstancedStaticMeshes=true
  bUsedWithMeshDeformer=true
  bUsedWithMorphTargets=true
  bUsedWithSkeletalMesh=true
  pixelDepthOffsetMode=PDOM_Legacy
  usedShadingModels=Default Lit
 }
 nodes [
  n1 = MaterialFunctionCall() {materialFunction=&BaseFN_CharacterFunction_PreFX}
  n2 = TextureObjectParameter() {
   parameterName=SLMask
   samplerType=SAMPLERTYPE_Masks
   texture=/Game/Characters/Player/Male/Medium/Bodies/M_MED_Dark_Storm_Armored/Skins/Super_Levels/Textures/T_M_MED_Dark_Storm_SuperLevels_Armored_Body_SL_FX
  }
  n3 = MaterialFunctionCall!(
   Material=MaterialFunctionCall!(In=@n2, Gmap=@n1.Gmap, Material=@n1.Gmap) {materialFunction=&/MF_SK_S21_Metal}.Result
   PreskinnedPosition=@n1.Pre skinned Local Position
   In=@n2
  ) {materialFunction=&/MF_SK_S21_Reality}
  n4 = GetMaterialAttributes(None=@n3.Result) {attributeGetTypes [#1]}
  n5 = TextureSample(Tex=@n2) {
   samplerType=SAMPLERTYPE_Masks
   texture=/Game/Characters/Player/Male/Medium/Bodies/M_MED_Lancelot/Skins/Superlevel/Textures/T_M_MED_Lancelot_SuperLevels_BodyAcc_SL_FX
  }
  n6 = SetMaterialAttributes!(
   MaterialAttributes=@n4.MaterialAttributes
   Normal=MaterialFunctionCall(Normal=@n4.Normal, Flatness=LinearInterpolate(Alpha=@n5.B) {constB=0.85}) {materialFunction=&FlattenNormal}.Result
  ) {attributeSetTypes [#1]}
  n7 = TextureSample(Tex=TextureObjectParameter() {parameterName=M, texture=/Game/Creative/BuildingActors/Props/Textures/CP_White_Texutre}) {
   samplerSource=SSM_Wrap_WorldGroupSettings
   samplerType=SAMPLERTYPE_Masks
   texture=&T_M_MED_Lancelot_SuperLevels_BodyAcc_SL_FX
  }
  n8 = ScalarParameter() {group=SL22, parameterName=killskin}
  MaterialFunctionCall(
   AO=@n1.AO
   Distance Blend=@n1.AO
   Fresnel=@n1.AO
   Gmap=@n1.AO
   Pre Skinned Local Position=@n1.AO
   Material=BlendMaterialAttributes(
    A=StaticSwitchParameter(True=@n6, False=@n3.Result) {parameterName=Lancelot Cape}
    B=@n1.Result
    Alpha=LinearInterpolate(B=@n7.B, Alpha=@n8) {constB=0}
   )
  ) {materialFunction=&BaseFN_CharacterFunction_PostFX}
  OneMinus(None=@n8)
  TextureSample() {
   samplerType=SAMPLERTYPE_Masks
   texture=&T_M_MED_Lancelot_SuperLevels_BodyAcc_SL_FX
  }
 ]
}
